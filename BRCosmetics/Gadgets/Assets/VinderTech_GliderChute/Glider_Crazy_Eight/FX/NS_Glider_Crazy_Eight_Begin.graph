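# Niagara system for the Crazy Eight glider's begin effect, rendering flattened propeller motion-blur meshes at a sampled socket.
# One delayed-burst emitter follows the socket and adds a yaw spin accumulated from a system-level integrator value.
# Three mesh renderers are fed by particle color and two mesh-attribute modules, with alpha curves fading in on system age once the spawn delay passes.
# User.Variant selects tints and fade intensity, and a front-end modifier adjusts the socket offset.
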NiagaraSystem /BRCosmetics/Gadgets/Assets/VinderTech_GliderChute/Glider_Crazy_Eight/FX/NS_Glider_Crazy_Eight_Begin
dependencies {
 usedDataInterfaces [&NiagaraDataInterfaceCurve, &NiagaraDataInterfaceSocketReader]
 usedDynamicInputs [&MakeCustomFloatFromBool, &SetBoolByIntComparison]
 usedModules [
  &self:SystemSpawnScript.SystemScriptSource.SystemScriptGraph.NiagaraNodeAssignment_3.SetVariables_#1
  &SystemState
  &FN_Integrator
 ]
}
emitters {MotionBlur_Base {
  emitterUpdateScript [
   &EmitterState(
    Inactive Response=::NewEnumerator1
    Life Cycle Mode=::NewEnumerator1
    Loop Behavior=::NewEnumerator1
    Loop Delay=1.85
    Loop Duration Mode=::NewEnumerator1
    UseLoopDelay=true
   )
   &SpawnBurst_Instantaneous(Use Spawn Probability=false, UseLoopCountLimit=false)
  ]
  particleSpawnScript [
   &InitializeParticle(
    Lifetime=5
    Mesh Scale=<1 1 0.01>
    Mesh Scale Mode=::NewEnumerator4
    Position Offset=<0 0 50>
    UsePositionOffset=false
    Write Material Random=true
   )
   /SetVariables() {isSetParameters=true}
  ]
  particleUpdateScript [
   &ParticleState(Kill Particles When Lifetime Has Elapsed=false, Let Infinitely Lived Particles Die When Emitter Deactivates=true)
   &SocketLocation(
    Orientation Sampling=::NewEnumerator1
    SamplePosOffset=true
    Sampled Orientation Coordinate Space:ENiagaraCoordinateSpace=::Local
    Sampled Position Offset=&FN_FrontEndVector3Modifier(Gameplay=<0 0 0>, Front End=<0 0 10>)
    Sampled Position Offset Coordinate Space:ENiagaraSkelMeshTransforms=::NewEnumerator2
    Socket Reader=$System.Socket Reader
    Socket Sampling Mode=::NewEnumerator3
    VelocitySpeedLimit=false
   )
   /SetVariables(Particles.PropellerSpeedIncrement:NiagaraFloat=&Add_Float(A=$Particles.PropellerSpeedIncrement, B=&Multiply_Float(A=$Engine.DeltaTime, B=$System.FN_Integrator.Value))) {isSetParameters=true}
   /SetVariables(Particles.MeshOrientation=&MultiplyQuaternion(
     Quaternion A=$Output.SocketLocation.SampledRotation
     Quaternion B=&MakeQuaternion(
      Quaternion From=1
      Axis=<1 0 0>
      Angle=0
      Yaw=$Particles.PropellerSpeedIncrement
      From=?
      To=?
      Angle Type=::NewEnumerator1
      Coordinate Space=::Local
      Normalize Vectors=false
     )
    )) {isSetParameters=true}
   &ScaleColor(
    Scale Alpha=&FloatFromCurve(
     FloatCurve=DataInterface() {propertyValues {Curve {keys [<at 0 0 RCIM_Cubic>,<at 0.6 1 RCIM_Cubic>]}}}
     CurveIndex=&Max_Float(A:NiagaraFloat=&Subtract_Float(A=$System.Age, B=1.85), B:NiagaraFloat=0)
     Scale Curve=&MergeHelper_FloatSwitch(0=1, 1=0.5, 2=1, 3=?, Index=&MakeIntFromFloat(Float=$User.Variant), Num=::NewEnumerator1)
    )
    Scale RGB=&MakeVectorFromLinearColorRGB(Color=&MergeHelper_ColorSwitch(
      0=<rgba 0.013 0.0186 0.0609 1>
      1=<rgba 0.5719 0.584 0.7 1>
      2=<rgba 0.013 0.0186 0.0609 1>
      3=?
      Index=&MakeIntFromFloat(Float=$User.Variant)
      Num=::NewEnumerator1
     ))
    ScaleA=true
    ScaleRGB=true
   )
   &DynamicMaterialParameters(
    Index 0 Param 2=-15
    Index 0 Param 3=1
    Index 0 Param 4=1
    Module.Index0Float0Write=true
    Module.Index0Float1Write=true
    Module.Index0Float2Write=true
    Module.Index0Float3Write=true
   )
   &FN_Mesh_Attributes(
    Color=&MultiplyLinearColor(
     Linear Color=&MergeHelper_ColorSwitch(
      0=<rgba 0.2918 0.3185 0.5149 1>
      1=<rgba 0.055 0.0697 0.15 1>
      2=<rgba 0.3916 0.3592 0.1975 0.5304>
      3=?
      Index=&MakeIntFromFloat(Float=$User.Variant)
      Num=::NewEnumerator1
     )
     Scale Factor=&MakeVector4FromVectorAndFloat(
      "Vector (RGB)"=<1 1 1>
      "Float (Alpha)"=&FloatFromCurve(
       FloatCurve=DataInterface() {propertyValues {Curve {keys [<at 0 0 RCIM_Cubic>,<at 0.6 1 RCIM_Cubic>]}}}
       CurveIndex=&Max_Float(A:NiagaraFloat=&Subtract_Float(A=$System.Age, B=1.85), B:NiagaraFloat=0)
      )
     )
    )
    Create Color Attribute=true
    Create Scale Attribute=true
    Module Output=::NewEnumerator1
    Scale=<1 1 0.01>
   )
   &FN_Mesh_Attributes(
    Color=&MultiplyLinearColor(
     Linear Color=$Particles.Color
     Scale Factor=&MakeVector4FromVectorAndFloat(
      "Vector (RGB)"=<10 10 10>
      "Float (Alpha)"=&FloatFromCurve(
       FloatCurve=DataInterface() {propertyValues {Curve {keys [<at 0 0 RCIM_Cubic>,<at 0.6 1 RCIM_Cubic>]}}}
       CurveIndex=&Max_Float(A:NiagaraFloat=&Subtract_Float(A=$System.Age, B=1.85), B:NiagaraFloat=0)
       Scale Curve=&MergeHelper_FloatSwitch(0=9, 1=15, 2=9, 3=?, Index=&MakeIntFromFloat(Float=$User.Variant), Num=::NewEnumerator1)
      )
     )
    )
    Create Color Attribute=true
    Create Scale Attribute=true
    Module Output=::NewEnumerator1
    Scale=<1.025 1.025 0.01>
   ) {module=FN_Mesh_Attributes001}
  ]
  rendererClasses [&NiagaraMeshRendererProperties]
  renderers [
   {rendererClass=&NiagaraMeshRendererProperties, rendererIndex=0}
   {rendererClass=&NiagaraMeshRendererProperties, rendererIndex=1}
   {rendererClass=&NiagaraMeshRendererProperties, rendererIndex=2}
  ]
 }}
summary {
 systemName=NS_Glider_Crazy_Eight_Begin
 emitters [{emitterName=MotionBlur_Base, rendererClasses [&NiagaraMeshRendererProperties]}]
 userVariables [
  User.bIsFrontEnd {type=NiagaraBool}
  User.bIsFrontEndPreview {type=NiagaraBool}
  User.ForwardDot {type=NiagaraFloat, defaultValue=0, description=""}
  User.RightDot {type=NiagaraFloat, defaultValue=0, description=""}
  User.Variant {type=NiagaraFloat, defaultValue=1, description=""}
 ]
}
